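; Winter variant of the drivable Quad (Antelope) vehicle Blueprint, adding a gameplay-tag reaction to damage.
; When damaged, it casts the damage instigator to a Fort player controller and loops over all passengers.
; For each passenger matching the SilkwalkerTagger tag, it adds NegativeTag as a loose gameplay tag, targeting the instigator's controller rather than the passenger.
; The construction script only calls its parent.
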(event EventOnDamaged
  (params (Damage :type Float%20%28single-precision%29) (DamageTags :type Gameplay%20Tag%20Container%20Structure%20%28by%20ref%29) (EffectContext :type Gameplay%20Effect%20Context%20Handle%20Structure%20%28by%20ref%29) (EventInstigator :type Controller%20Object%20Reference) (DamageCauser :type Actor%20Object%20Reference))
  (bind _asfort_player_controller (Utilities|Casting|CastToFortPlayerController EventInstigator))
  (bind _array_element (Utilities|Array|ForEachLoopwithBreak (Vehicle|GetAllPassengers))
    (:LoopBody
      (if (GameplayTags|HasMatchingGameplayTag _array_element (Variables|Default|GetSilkwalkerTagger))
        (Ability|GameplayEffect|AddLooseGameplayTags _asfort_player_controller (Variables|Default|GetNegativeTag))))))

(fn ConstructionScript
  (|CallParentFunction))
